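# Sound cue for opening a WinterFest ice machine chest, using item-interact attenuation and a loot reward sound class.
# A mixer layers one of four random ice machine open sounds with one of four quieter, briefly delayed treasure chest sweeteners.
# Each layer has its own random pitch modulation.
# A concurrency override blocks new instances once the limit is reached.
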SoundCue /Game/Athena/Sounds/Items/WinterFest/IceMachine_Chest_Open_Cue
root=Mixer_0
nodes [
 {id=Delay_0, children [Random_1], properties {delayMax=0.1, delayMin=0.1}}
 {id=Mixer_0, children [Modulator_1,Modulator_3], properties {inputVolume [1.5,0.7]}}
 {id=Modulator_1, children [Random_0], properties {pitchMax=1.2, volumeMax=1, volumeMin=1}}
 {id=Modulator_3, children [Delay_0], properties {pitchMax=1.5, pitchMin=1, volumeMax=0.7, volumeMin=0.7}}
 {
  id=Random_0
  children [WavePlayer_5,WavePlayer_2,WavePlayer_3,WavePlayer_4]
  properties {bShouldExcludeFromBranchCulling=true, weights [1,1,1,1]}
 }
 {
  id=Random_1
  children [WavePlayer_16, WavePlayer_17, WavePlayer_18, WavePlayer_19]
  properties {bShouldExcludeFromBranchCulling=true, weights [1,1,1,1]}
 }
 {id=WavePlayer_16, properties {soundWaveAssetPtr=/Game/Sounds/Foley_Loot/Containers/Treasure_Chest/treasure_chest_sweet_shorter_01}}
 {id=WavePlayer_17, properties {soundWaveAssetPtr=/Game/Sounds/Foley_Loot/Containers/Treasure_Chest/treasure_chest_sweet_shorter_02}}
 {id=WavePlayer_18, properties {soundWaveAssetPtr=/Game/Sounds/Foley_Loot/Containers/Treasure_Chest/treasure_chest_sweet_shorter_03}}
 {id=WavePlayer_19, properties {soundWaveAssetPtr=/Game/Sounds/Foley_Loot/Containers/Treasure_Chest/treasure_chest_sweet_shorter_04}}
 {id=WavePlayer_2, properties {soundWaveAssetPtr=&/IceMachine_Open_02}}
 {id=WavePlayer_3, properties {soundWaveAssetPtr=&/IceMachine_Open_03}}
 {id=WavePlayer_4, properties {soundWaveAssetPtr=&/IceMachine_Open_04}}
 {id=WavePlayer_5, properties {soundWaveAssetPtr=&/IceMachine_Open_01}}
]
properties {
 attenuationSettings=/Game/Sounds/Attenuation/Fort_Item_Interact_Att
 bOverrideConcurrency=true
 maxDistance=1850
 soundClassObject=&Fort_Loot_Reward_PSM
 volumeMultiplier=1
 concurrencyOverrides {resolutionRule=PreventNew}
}
