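# Sound cue for searching the WinterFest ice machine chest, played with item-interact attenuation in the loot-collect sound class.
# The first two quality levels mix a random ice machine search clip, faded out over about four seconds by an enveloper, with a random gear-search foley clip.
# The last two quality levels play a single modulated ice machine clip instead.
# A concurrency override prevents new instances once the limit is reached.
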SoundCue /Game/Athena/Sounds/Items/WinterFest/IceMachine_Chest_Search_cue
root=QualityLevel_0
nodes [
 {
  id=Enveloper_1
  children [Random_0]
  properties {
   pitchCurve {editorCurveData {keys [{time=0}]}}
   volumeCurve {editorCurveData {keys [
      {arriveTangent=-0.0387, time=0, value=1}
      {arriveTangent=-0.264, interpMode=RCIM_Cubic, leaveTangent=-0.264, time=2.7563, value=0.8772}
      <at 3.7875 0>
     ]}}
  }
 }
 {id=Mixer_2, children [Enveloper_1,Random_13], properties {inputVolume [0.6,1]}}
 {id=Modulator_0, children [WavePlayer_8]}
 {id=QualityLevel_0, children [Mixer_2,Mixer_2,Modulator_0,Modulator_0]}
 {id=Random_0, children [WavePlayer_5,WavePlayer_6], properties {weights [1,1]}}
 {id=Random_13, children [WavePlayer_58, WavePlayer_59, WavePlayer_60], properties {weights [1,1,1]}}
 {id=WavePlayer_5, properties {soundWaveAssetPtr=&/IceMachine_Search_01}}
 {id=WavePlayer_58, properties {soundWaveAssetPtr=/Game/Sounds/Foley_Loot/Containers/Misc/Prop_Gear_Search_04}}
 {id=WavePlayer_59, properties {soundWaveAssetPtr=/Game/Sounds/Foley_Loot/Containers/Misc/Prop_Gear_Search_06}}
 {id=WavePlayer_6, properties {soundWaveAssetPtr=&/IceMachine_Search_02}}
 {id=WavePlayer_60, properties {soundWaveAssetPtr=/Game/Sounds/Foley_Loot/Containers/Misc/Prop_Gear_Search_01}}
 {id=WavePlayer_8, properties {soundWaveAssetPtr=&IceMachine_Search_01}}
]
properties {
 attenuationSettings=/Game/Sounds/Attenuation/Fort_Item_Interact_Mono_Att
 bOverrideConcurrency=true
 maxDistance=1500
 soundClassObject=&Fort_Loot_Collect
 volumeMultiplier=1
 concurrencyOverrides {resolutionRule=PreventNew}
}
