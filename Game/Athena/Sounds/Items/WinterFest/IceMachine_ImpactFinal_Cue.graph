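# Sound cue for the final impact of a WinterFest ice machine, with medium-object demolition attenuation and four-voice impact concurrency.
# Three pitch-modulated layers are mixed: metal barrel impact, metal machine impact and a slightly quieter ice machine explosion.
# A quality switch feeds its first two levels a mixer with random wave variations per layer, and its last two a mixer with one fixed wave per layer.
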SoundCue /Game/Athena/Sounds/Items/WinterFest/IceMachine_ImpactFinal_Cue
root=QualityLevel_0
nodes [
 {id=Mixer_0, children [Modulator_1,Modulator_0,Modulator_2], properties {inputVolume [1,1,0.8]}}
 {id=Mixer_1, children [Modulator_5,Modulator_6,Modulator_4], properties {inputVolume [1,1,0.8]}}
 {id=Modulator_0, children [Random_2], properties {pitchMax=1.2, pitchMin=0.92}}
 {id=Modulator_1, children [Random_1]}
 {id=Modulator_2, children [Random_0], properties {pitchMin=0.98}}
 {id=Modulator_4, children [WavePlayer_7], properties {pitchMin=0.98}}
 {id=Modulator_5, children [WavePlayer_10]}
 {id=Modulator_6, children [WavePlayer_12], properties {pitchMax=1.2, pitchMin=0.92}}
 {id=QualityLevel_0, children [Mixer_0,Mixer_0,Mixer_1,Mixer_1]}
 {id=Random_0, children [WavePlayer_0,WavePlayer_1,WavePlayer_2], properties {weights [1,1,1]}}
 {id=Random_1, children [WavePlayer_3,WavePlayer_4], properties {weights [1,1]}}
 {id=Random_2, children [WavePlayer_5,WavePlayer_6], properties {weights [1,1]}}
 {id=WavePlayer_0, properties {soundWaveAssetPtr=&/IceMachine_Explode_01}}
 {id=WavePlayer_1, properties {soundWaveAssetPtr=&/IceMachine_Explode_02}}
 {id=WavePlayer_10, properties {soundWaveAssetPtr=/Game/Sounds/Fort_Impact_Sounds/Demolition/Metal/Metal_Barrel_ImpactFinal02}}
 {id=WavePlayer_12, properties {soundWaveAssetPtr=/Game/Sounds/Fort_Impact_Sounds/Demolition/Metal/Metal_Machine_Impact02}}
 {id=WavePlayer_2, properties {soundWaveAssetPtr=&/IceMachine_Explode_03}}
 {id=WavePlayer_3, properties {soundWaveAssetPtr=&Metal_Barrel_ImpactFinal02}}
 {id=WavePlayer_4, properties {soundWaveAssetPtr=/Game/Sounds/Fort_Impact_Sounds/Demolition/Metal/Metal_Barrel_ImpactFinal01}}
 {id=WavePlayer_5, properties {soundWaveAssetPtr=&Metal_Machine_Impact02}}
 {id=WavePlayer_6, properties {soundWaveAssetPtr=/Game/Sounds/Fort_Impact_Sounds/Demolition/Metal/Metal_Machine_Impact03}}
 {id=WavePlayer_7, properties {soundWaveAssetPtr=&IceMachine_Explode_01}}
]
properties {
 attenuationSettings=/Game/Sounds/Attenuation/Fort_Demolition_Destroy_Medium_Sized_Object_Att
 maxDistance=4000
 soundClassObject=&Fort_Impacts_PSM
 volumeMultiplier=0.34
 concurrencyOverrides {maxCount=4, resolutionRule=StopOldest}
 concurrencySet [&CONC_Object_Impacts_FourMax]
}
