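# Sound cue for impacts on the WinterFest ice machine, built from reused barbell, metal barrel and machine impact samples.
# Each variant mixes those three layers, with random pitch on the whole mix and wider variation on the machine layer.
# A quality node gives the first two levels randomized samples per layer and the last two a fixed sample per layer.
# Demolition-hit attenuation and stop-oldest concurrency limit range and overlap.
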SoundCue /Game/Athena/Sounds/Items/WinterFest/IceMachine_Impact_Cue
root=QualityLevel_1
nodes [
 {id=Mixer_0, children [Random_27,Random_1,Modulator_4], properties {inputVolume [1,1,1]}}
 {id=Mixer_1, children [WavePlayer_9, WavePlayer_13, Modulator_6], properties {inputVolume [1,1,1]}}
 {id=Modulator_2, children [Mixer_0], properties {pitchMax=1.07}}
 {id=Modulator_4, children [Random_2], properties {pitchMax=1.2, pitchMin=0.92}}
 {id=Modulator_5, children [Mixer_1], properties {pitchMax=1.07}}
 {id=Modulator_6, children [WavePlayer_18], properties {pitchMax=1.2, pitchMin=0.92}}
 {id=QualityLevel_1, children [Modulator_2,Modulator_2,Modulator_5,Modulator_5]}
 {id=Random_1, children [WavePlayer_2,WavePlayer_4,WavePlayer_5], properties {weights [1,1,1]}}
 {id=Random_2, children [WavePlayer_6,WavePlayer_8], properties {weights [1,1]}}
 {id=Random_27, children [WavePlayer_78, WavePlayer_79, WavePlayer_80], properties {weights [1,1,1]}}
 {id=WavePlayer_13, properties {soundWaveAssetPtr=/Game/Sounds/Fort_Impact_Sounds/Demolition/Metal/Metal_Barrel_Impact04}}
 {id=WavePlayer_18, properties {soundWaveAssetPtr=/Game/Sounds/Fort_Impact_Sounds/Demolition/Metal/Metal_Machine_Impact02}}
 {id=WavePlayer_2, properties {soundWaveAssetPtr=&Metal_Barrel_Impact04}}
 {id=WavePlayer_4, properties {soundWaveAssetPtr=/Game/Sounds/Fort_Impact_Sounds/Demolition/Metal/Metal_Barrel_Impact05}}
 {id=WavePlayer_5, properties {soundWaveAssetPtr=/Game/Sounds/Fort_Impact_Sounds/Demolition/Metal/Metal_Barrel_Impact06}}
 {id=WavePlayer_6, properties {soundWaveAssetPtr=&Metal_Machine_Impact02}}
 {id=WavePlayer_78, properties {soundWaveAssetPtr=/Game/Sounds/Fort_Impact_Sounds/Demolition/Gym/Gym_Barbell_Impact04}}
 {id=WavePlayer_79, properties {soundWaveAssetPtr=/Game/Sounds/Fort_Impact_Sounds/Demolition/Gym/Gym_Barbell_Impact01}}
 {id=WavePlayer_8, properties {soundWaveAssetPtr=/Game/Sounds/Fort_Impact_Sounds/Demolition/Metal/Metal_Machine_Impact03}}
 {id=WavePlayer_80, properties {soundWaveAssetPtr=/Game/Sounds/Fort_Impact_Sounds/Demolition/Gym/Gym_Barbell_Impact02}}
 {id=WavePlayer_9, properties {soundWaveAssetPtr=&Gym_Barbell_Impact04}}
]
properties {
 attenuationSettings=/Game/Sounds/Attenuation/Fort_Demolition_Hit_Small_Sized_Object_Att
 maxDistance=4000
 soundClassObject=&Fort_Impacts
 volumeMultiplier=0.34
 concurrencyOverrides {maxCount=3, resolutionRule=StopOldest}
 concurrencySet [&CONC_Object_Impacts_FourMax]
}
