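# Sound cue for the WinterFest ice machine, apparently for repeat loot attempts, layering a "no more loot" UI sound with a machine search sound.
# Search sounds fade out through an Enveloper, get slight random pitch and volume modulation, then mix with the UI sound.
# A quality switch gives its first two slots two random search variants and its last two just one.
# Concurrency allows one instance, blocking new plays, with item-interact attenuation.
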SoundCue /Game/Athena/Sounds/Items/WinterFest/IceMachine_LootRepeat_Cue
root=QualityLevel_0
nodes [
 {
  id=Enveloper_0
  children [Random_0]
  properties {
   pitchCurve {editorCurveData {keys [{time=0}]}}
   volumeCurve {editorCurveData {keys [{arriveTangent=-1.3539, interpMode=RCIM_Cubic, leaveTangent=-1.3539, tangentMode=RCTM_User, time=0, value=1}, <at 2 0 RCIM_Cubic>]}}
  }
 }
 {
  id=Enveloper_1
  children [WavePlayer_3]
  properties {
   pitchCurve {editorCurveData {keys [{time=0}]}}
   volumeCurve {editorCurveData {keys [{arriveTangent=-1.3539, interpMode=RCIM_Cubic, leaveTangent=-1.3539, tangentMode=RCTM_User, time=0, value=1}, <at 2.0174 -0.0448 RCIM_Cubic>]}}
  }
 }
 {id=Mixer_0, children [WavePlayer_1,Modulator_0], properties {inputVolume [0.8,1]}}
 {id=Mixer_3, children [WavePlayer_14, Modulator_2], properties {inputVolume [0.8,1]}}
 {id=Modulator_0, children [Enveloper_1], properties {pitchMax=1.07, pitchMin=0.92, volumeMax=1.07, volumeMin=0.92}}
 {id=Modulator_2, children [Enveloper_0], properties {pitchMax=1.07, pitchMin=0.92, volumeMax=1.07, volumeMin=0.92}}
 {id=QualityLevel_0, children [Mixer_3,Mixer_3,Mixer_0,Mixer_0]}
 {id=Random_0, children [WavePlayer_5,WavePlayer_6], properties {weights [1,1]}}
 {id=WavePlayer_1, properties {soundWaveAssetPtr=/Game/Sounds/Fort_UI_Sounds/Gameplay/Fort_UI_NoMoreLoot}}
 {id=WavePlayer_14, properties {soundWaveAssetPtr=&Fort_UI_NoMoreLoot}}
 {id=WavePlayer_3, properties {soundWaveAssetPtr=&/IceMachine_Search_01}}
 {id=WavePlayer_5, properties {soundWaveAssetPtr=&IceMachine_Search_01}}
 {id=WavePlayer_6, properties {soundWaveAssetPtr=&/IceMachine_Search_02}}
]
properties {
 attenuationSettings=/Game/Sounds/Attenuation/Fort_Item_Interact_Mono_Att
 bOverrideConcurrency=true
 maxDistance=1500
 soundClassObject=&Fort_Loot_Collect
 volumeMultiplier=1
 concurrencyOverrides {maxCount=1, resolutionRule=PreventNew}
}
